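# Animation Blueprint for Burt, a rideable quadruped prey creature, covering wild and ridden movement plus an abduction pose.
# Separate Locomotion and RidingLocomotion state machines handle idle, turns, movement, jumps and swimming, with sleep only when unridden.
# Additives layer head tilt, turn lean, body roll and landings; idle comes from a linked anim layer.
# The selected pose passes through montage slots, weighted head look-at, quadruped slope warping, leg IK and inertialization.
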AnimBlueprint /Irwin/AI/Prey/Burt/Animations/Burt_AnimBP
target_skeleton=/Irwin/AI/Base/Burt_Mammal_Invis_Skeleton
graphs [
 AnimGraph {nodes [
   AnimGraphNode_StateMachine_1 = LocomotionPoseAdditives() {is=title AnimGraphNode_StateMachine}
   AnimGraphNode_StateMachine_2 = Locomotion() {is=title AnimGraphNode_StateMachine}
   n1 = Apply Additive(
    Base=@AnimGraphNode_StateMachine_2
    Additive=@AnimGraphNode_StateMachine_1
    Alpha=Get LocomotionPoseAdditiveAlpha() -> (LocomotionPoseAdditiveAlpha:float)
   ) {alphaBoolBlend {blendOption=Linear}}
   AnimGraphNode_StateMachine_0 = EntryAdditive() {is=title AnimGraphNode_StateMachine}
   AnimGraphNode_StateMachine_3 = RidingLocomotion() {is=title AnimGraphNode_StateMachine}
   n2 = Apply Additive(Base=@AnimGraphNode_StateMachine_3, Additive=@AnimGraphNode_StateMachine_0) {alphaBoolBlend {blendOption=Linear}}
   n3 = Break Riding Animation Data!(RidingAnimationData:"Riding Animation Data Structure (by ref)"=Get RidingAnimationData() -> (RidingAnimationData:Riding Animation Data Structure)) -> (
    bIsRidden:bool=~
    bIsTurningInPlace:bool=~
    bIsSprinting:bool=~
    bIsBeingPetted:bool=~
    bIsMovingBackwards:bool=~
    TurnInPlaceRotationSpeed:float=~
    TurnInPlaceAngleDelta:float=~
    RiderReferentialYaw:float=~
   ) {is=title K2Node_BreakStruct}
   n4 = Slot 'DefaultSlot'(Source=Blend Poses by bool(
     BlendPose_0=Sequence Player 'Burt_Abduction'() {playRate=1.2, sequence=&/Burt_Abduction}
     BlendPose_1=Blend Poses by bool(BlendPose_0=@n2, BlendPose_1=@n1, bActiveValue=@n3) {blendType=Linear}
     BlendTime_0=0.5
     BlendTime_1=0.5
     bActiveValue=Get bIsAbducted() -> (bIsAbducted:bool)
    ) {blendType=Linear, transitionType=Inertialization})
   n5 = Weighted Look At - Bone: BipedHeadNeck_A_Head_C!(
    ComponentPose:Component Space Pose Link Structure=Local To Component(LocalPose=Slot 'FullBodyAdditiveSlot'(Source=@n4) {slotName=FullBodyAdditiveSlot})
    Alpha:float=Get LookAtAlpha() -> (LookAtAlpha:float)
    LookAtLocation:Vector=Get AimTargetLocation() -> (AimTargetLocation:Vector)
   ) -> (Pose:Component Space Pose Link Structure) {
    is=title FortAnimGraphNode_WeightedLookAt
    interpolationTime=3
    interpolationType=EaseInOutExponent3
    lookAtClamp=45
    alphaBoolBlend {blendOption=Linear}
    boneToModify {boneName=BipedHeadNeck_A_Head_C}
    boneWeightsToApply [{weight=0.3}, {weight=0.3}, {weight=0.25}, {weight=0.15}]
    lookAt_Axis {axis=<1 0>}
    lookUp_Axis {axis=<0 1>}
   }
   n6 = Quadruped Slope Warping!(ComponentPose:Component Space Pose Link Structure=@n5, Alpha:float=1) -> (Pose:Component Space Pose Link Structure) {
    is=title FortAnimGraphNode_QuadrupedSlopeWarping
    maxStepHeight=35
    numSpineBones=4
    alphaBoolBlend {blendOption=Linear}
    armsFeetDefinitions [
     {numBonesInLimb=3, fKFootBone {boneName=PawedArm_A_Ball_L}, iKFootBone {boneName=IK_Foot_Front_L}}
     {numBonesInLimb=3, fKFootBone {boneName=PawedArm_A_Ball_R}, iKFootBone {boneName=IK_Foot_Front_R}}
    ]
    armsIKFootRootBone {boneName=IK_Foot_Front_Root}
    armsRootBone {boneName=QuadSpine_A_Chest_C}
    feetDefinitions [
     {numBonesInLimb=3, fKFootBone {boneName=PawedLeg_A_Ball_L}, iKFootBone {boneName=IK_Foot_Rear_L}}
     {numBonesInLimb=3, fKFootBone {boneName=PawedLeg_A_Ball_R}, iKFootBone {boneName=IK_Foot_Rear_R}}
    ]
    iKFootRootBone {boneName=IK_Foot_Rear_Root}
    pelvisBone {boneName=Leaf_Hips_Leaf_C}
    skeletonRootBone {boneName=root}
   }
   n7 = Leg IK!(ComponentPose:Component Space Pose Link Structure=@n6, Alpha:float=1) -> (Pose:Component Space Pose Link Structure) {
    is=title AnimGraphNode_LegIK
    alphaBoolBlend {blendOption=Linear}
    legsDefinition [
     {
      bEnableKneeTwistCorrection=true
      bEnableRotationLimit=false
      footBoneForwardAxis=Y
      hingeRotationAxis=None
      minRotationAngle=15
      numBonesInLimb=2
      twistOffsetCurveName=None
      fKFootBone {boneName=PawedLeg_A_Ball_L}
      iKFootBone {boneName=IK_Foot_Rear_L}
     }
     {
      bEnableKneeTwistCorrection=true
      bEnableRotationLimit=false
      footBoneForwardAxis=Y
      hingeRotationAxis=None
      minRotationAngle=15
      numBonesInLimb=2
      twistOffsetCurveName=None
      fKFootBone {boneName=PawedLeg_A_Ball_R}
      iKFootBone {boneName=IK_Foot_Rear_R}
     }
     {
      bEnableKneeTwistCorrection=true
      bEnableRotationLimit=false
      footBoneForwardAxis=Y
      hingeRotationAxis=None
      minRotationAngle=15
      numBonesInLimb=2
      twistOffsetCurveName=None
      fKFootBone {boneName=PawedArm_A_Ball_R}
      iKFootBone {boneName=IK_Foot_Front_R}
     }
     {
      bEnableKneeTwistCorrection=true
      bEnableRotationLimit=false
      footBoneForwardAxis=Y
      hingeRotationAxis=None
      minRotationAngle=15
      numBonesInLimb=2
      twistOffsetCurveName=None
      fKFootBone {boneName=PawedArm_A_Ball_L}
      iKFootBone {boneName=IK_Foot_Front_L}
     }
    ]
   }
   n8 = Component To Local(ComponentPose=@n7)
   Output Pose(Result=Inertialization!(Source:Pose Link Structure=Slot 'SequencerDefault'(Source=@n8) {slotName=SequencerDefault}) -> (Pose:Pose Link Structure) {is=title AnimGraphNode_Inertialization}) {layerGroup=None}
   n9 = Fauna_Idle_AnimInterface - Fauna Idle!(InPose Idle:Pose Link Structure) -> (Pose:Pose Link Structure) {
    is=title AnimGraphNode_LinkedAnimLayer
    bPropagateNotifiesToLinkedInstances=true
    bReceiveNotifiesFromLinkedInstances=true
    interface=/Irwin/AI/Interfaces/Fauna_Idle_AnimInterface.Fauna_Idle_AnimInterface_C
    layer=Fauna Idle
   }
   Save cached pose 'FaunaIdlePose'(Pose=@n9) {cacheName=FaunaIdlePose}
  ]}
 EntryAdditive {
  entry=AddRef
  path=AnimGraph.AnimGraphNode_StateMachine_0.EntryAdditive
  ids=id
  nodes [
   AnimStateNode_0 = AddRef() {is=title AnimStateNode}
   AnimStateNode_1 = LandAdd() {is=title AnimStateNode, stateLeft {guid=#1, notifyName=LandAddExit}}
   AnimStateTransitionNode_0 = AnimStateTransitionNode() {crossfadeDuration=0.1, logicType=TLT_Inertialization}
   AnimStateTransitionNode_1 = AnimStateTransitionNode() {bAutomaticRuleBasedOnSequencePlayerInState=true, crossfadeDuration=0.4, logicType=TLT_Inertialization}
  ]
  links [
   {
    dst=LandAdd
    rule=AnimGraph.AnimGraphNode_StateMachine_0.EntryAdditive.AnimStateTransitionNode_0.Transition
    src=AddRef
    via=AnimStateTransitionNode_0
   }
   {
    dst=AddRef
    rule=AnimGraph.AnimGraphNode_StateMachine_0.EntryAdditive.AnimStateTransitionNode_1.Transition
    src=LandAdd
    via=AnimStateTransitionNode_1
   }
  ]
 }
 AddRef {
  path=AnimGraph.AnimGraphNode_StateMachine_0.EntryAdditive.AnimStateNode_0.AddRef
  nodes [Output Animation Pose(Result=Additive Identity Pose())]
 }
 LandAdd {
  path=AnimGraph.AnimGraphNode_StateMachine_0.EntryAdditive.AnimStateNode_1.LandAdd
  nodes [Output Animation Pose(Result="Sequence Player (additive) 'Burt_JumpLand_Moving_Add'"() {sequence=&/Burt_JumpLand_Moving_Add})]
 }
 Transition {path=AnimGraph.AnimGraphNode_StateMachine_0.EntryAdditive.AnimStateTransitionNode_0.Transition, nodes [Result()]}
 Transition {path=AnimGraph.AnimGraphNode_StateMachine_0.EntryAdditive.AnimStateTransitionNode_1.Transition, nodes [Result()]}
 LocomotionPoseAdditives {
  entry=Default
  path=AnimGraph.AnimGraphNode_StateMachine_1.LocomotionPoseAdditives
  ids=id
  nodes [
   AnimStateNode_0 = Default() {is=title AnimStateNode}
   AnimStateNode_1 = HeadTiltUp() {is=title AnimStateNode}
   AnimStateTransitionNode_0 = AnimStateTransitionNode() {crossfadeDuration=0.65, logicType=TLT_Inertialization}
   AnimStateTransitionNode_1 = AnimStateTransitionNode() {bAutomaticRuleBasedOnSequencePlayerInState=true, crossfadeDuration=0.8, logicType=TLT_Inertialization}
  ]
  links [
   {
    dst=HeadTiltUp
    rule=AnimGraph.AnimGraphNode_StateMachine_1.LocomotionPoseAdditives.AnimStateTransitionNode_0.Transition
    src=Default
    via=AnimStateTransitionNode_0
   }
   {
    dst=Default
    rule=AnimGraph.AnimGraphNode_StateMachine_1.LocomotionPoseAdditives.AnimStateTransitionNode_1.Transition
    src=HeadTiltUp
    via=AnimStateTransitionNode_1
   }
  ]
 }
 Default {
  path=AnimGraph.AnimGraphNode_StateMachine_1.LocomotionPoseAdditives.AnimStateNode_0.Default
  nodes [Output Animation Pose(Result=Additive Identity Pose())]
 }
 HeadTiltUp {
  path=AnimGraph.AnimGraphNode_StateMachine_1.LocomotionPoseAdditives.AnimStateNode_1.HeadTiltUp
  nodes [Output Animation Pose(Result="Sequence Player (additive) 'Burt_Locomotion_Additive_HeadTiltUp'"() {sequence=&/Burt_Locomotion_Additive_HeadTiltUp})]
 }
 Transition {
  path=AnimGraph.AnimGraphNode_StateMachine_1.LocomotionPoseAdditives.AnimStateTransitionNode_0.Transition
  nodes [Result(bCanEnterTransition=Get bEnableHeadTiltUpAdditive() -> (bEnableHeadTiltUpAdditive:bool))]
 }
 Transition {path=AnimGraph.AnimGraphNode_StateMachine_1.LocomotionPoseAdditives.AnimStateTransitionNode_1.Transition, nodes [Result()]}
 Locomotion {
  entry=Idle
  path=AnimGraph.AnimGraphNode_StateMachine_2.Locomotion
  ids=id
  nodes [
   AnimStateNode_0 = Idle()
   AnimStateNode_1 = Moving() {is=title AnimStateNode}
   AnimStateNode_2 = Swimming() {is=title AnimStateNode}
   AnimStateNode_3 = FallAsleep() {is=title AnimStateNode}
   AnimStateNode_4 = Sleeping() {is=title AnimStateNode}
   AnimStateNode_5 = JumpStart() {is=title AnimStateNode}
   AnimStateNode_6 = Land() {is=title AnimStateNode, stateEntered {guid=#2, notifyName=LandEnter}}
   AnimStateNode_7 = WakeUp() {is=title AnimStateNode}
   AnimStateNode_8 = Fall() {is=title AnimStateNode, stateLeft {guid=#3, notifyName=FallExit}}
   AnimStateTransitionNode_0 = AnimStateTransitionNode() {crossfadeDuration=0.4, logicType=TLT_Inertialization, priorityOrder=3}
   AnimStateTransitionNode_1 = AnimStateTransitionNode() {crossfadeDuration=0.1, logicType=TLT_Inertialization}
   AnimStateTransitionNode_10 = AnimStateTransitionNode()
   AnimStateTransitionNode_12 = AnimStateTransitionNode() {crossfadeDuration=0.1, logicType=TLT_Inertialization, priorityOrder=2}
   AnimStateTransitionNode_13 = AnimStateTransitionNode() {bAutomaticRuleBasedOnSequencePlayerInState=true, crossfadeDuration=0.4, logicType=TLT_Inertialization}
   AnimStateTransitionNode_14 = AnimStateTransitionNode() {crossfadeDuration=0.1, logicType=TLT_Inertialization}
   AnimStateTransitionNode_15 = AnimStateTransitionNode() {crossfadeDuration=0.1, logicType=TLT_Inertialization}
   AnimStateTransitionNode_16 = AnimStateTransitionNode() {crossfadeDuration=0.25, logicType=TLT_Inertialization, priorityOrder=2}
   AnimStateTransitionNode_17 = AnimStateTransitionNode()
   AnimStateTransitionNode_18 = AnimStateTransitionNode()
   AnimStateTransitionNode_19 = AnimStateTransitionNode() {bAutomaticRuleBasedOnSequencePlayerInState=true, crossfadeDuration=0.35, logicType=TLT_Inertialization, priorityOrder=100}
   AnimStateTransitionNode_2 = AnimStateTransitionNode() {crossfadeDuration=0.4, logicType=TLT_Inertialization, priorityOrder=100}
   AnimStateTransitionNode_21 = AnimStateTransitionNode() {crossfadeDuration=0.4, logicType=TLT_Inertialization, priorityOrder=100}
   AnimStateTransitionNode_22 = AnimStateTransitionNode() {crossfadeDuration=0.4, logicType=TLT_Inertialization, priorityOrder=100}
   AnimStateTransitionNode_25 = AnimStateTransitionNode() {crossfadeDuration=0.4, logicType=TLT_Inertialization}
   AnimStateTransitionNode_26 = AnimStateTransitionNode() {crossfadeDuration=0.4, logicType=TLT_Inertialization, priorityOrder=2}
   AnimStateTransitionNode_4 = AnimStateTransitionNode() {crossfadeDuration=0.35, logicType=TLT_Inertialization, priorityOrder=100}
   AnimStateTransitionNode_5 = AnimStateTransitionNode()
   AnimStateTransitionNode_6 = AnimStateTransitionNode() {crossfadeDuration=0.1, logicType=TLT_Inertialization}
   AnimStateTransitionNode_7 = AnimStateTransitionNode() {crossfadeDuration=0.4, logicType=TLT_Inertialization, priorityOrder=3}
   AnimStateTransitionNode_8 = AnimStateTransitionNode() {crossfadeDuration=4, logicType=TLT_Inertialization, priorityOrder=2}
   AnimStateTransitionNode_9 = AnimStateTransitionNode() {crossfadeDuration=0.4, logicType=TLT_Inertialization, priorityOrder=3}
  ]
  links [
   {
    dst=Land
    rule=AnimGraph.AnimGraphNode_StateMachine_2.Locomotion.AnimStateTransitionNode_12.AnimationTransitionGraph_0
    src=Fall
    via=AnimStateTransitionNode_12
   }
   {
    dst=Moving
    rule=AnimGraph.AnimGraphNode_StateMachine_2.Locomotion.AnimStateTransitionNode_15.Transition
    src=Fall
    via=AnimStateTransitionNode_15
   }
   {
    dst=Swimming
    rule=AnimGraph.AnimGraphNode_StateMachine_2.Locomotion.AnimStateTransitionNode_2.Transition
    src=Fall
    via=AnimStateTransitionNode_2
   }
   {
    dst=Sleeping
    rule=AnimGraph.AnimGraphNode_StateMachine_2.Locomotion.AnimStateTransitionNode_5.AnimationTransitionGraph_1
    src=FallAsleep
    via=AnimStateTransitionNode_5
   }
   {
    dst=Fall
    rule=AnimGraph.AnimGraphNode_StateMachine_2.Locomotion.AnimStateTransitionNode_26.Transition
    src=Idle
    via=AnimStateTransitionNode_26
   }
   {
    dst=FallAsleep
    rule=AnimGraph.AnimGraphNode_StateMachine_2.Locomotion.AnimStateTransitionNode_17.Transition
    src=Idle
    via=AnimStateTransitionNode_17
   }
   {
    dst=JumpStart
    rule=AnimGraph.AnimGraphNode_StateMachine_2.Locomotion.AnimStateTransitionNode_6.Transition
    src=Idle
    via=AnimStateTransitionNode_6
   }
   {
    dst=Moving
    rule=AnimGraph.AnimGraphNode_StateMachine_2.Locomotion.AnimStateTransitionNode_0.Transition
    src=Idle
    via=AnimStateTransitionNode_0
   }
   {
    dst=Swimming
    rule=AnimGraph.AnimGraphNode_StateMachine_2.Locomotion.AnimStateTransitionNode_22.Transition
    src=Idle
    via=AnimStateTransitionNode_22
   }
   {
    dst=Fall
    rule=AnimGraph.AnimGraphNode_StateMachine_2.Locomotion.AnimStateTransitionNode_13.Transition
    src=JumpStart
    via=AnimStateTransitionNode_13
   }
   {
    dst=Idle
    rule=AnimGraph.AnimGraphNode_StateMachine_2.Locomotion.AnimStateTransitionNode_19.Transition
    src=Land
    via=AnimStateTransitionNode_19
   }
   {
    dst=JumpStart
    rule=AnimGraph.AnimGraphNode_StateMachine_2.Locomotion.AnimStateTransitionNode_16.Transition
    src=Land
    via=AnimStateTransitionNode_16
   }
   {
    dst=Moving
    rule=AnimGraph.AnimGraphNode_StateMachine_2.Locomotion.AnimStateTransitionNode_1.Transition
    src=Land
    via=AnimStateTransitionNode_1
   }
   {
    dst=Swimming
    rule=AnimGraph.AnimGraphNode_StateMachine_2.Locomotion.AnimStateTransitionNode_9.Transition
    src=Land
    via=AnimStateTransitionNode_9
   }
   {
    dst=Fall
    rule=AnimGraph.AnimGraphNode_StateMachine_2.Locomotion.AnimStateTransitionNode_8.Transition
    src=Moving
    via=AnimStateTransitionNode_8
   }
   {
    dst=Idle
    rule=AnimGraph.AnimGraphNode_StateMachine_2.Locomotion.AnimStateTransitionNode_4.Transition
    src=Moving
    via=AnimStateTransitionNode_4
   }
   {
    dst=JumpStart
    rule=AnimGraph.AnimGraphNode_StateMachine_2.Locomotion.AnimStateTransitionNode_14.Transition
    src=Moving
    via=AnimStateTransitionNode_14
   }
   {
    dst=Swimming
    rule=AnimGraph.AnimGraphNode_StateMachine_2.Locomotion.AnimStateTransitionNode_7.Transition
    src=Moving
    via=AnimStateTransitionNode_7
   }
   {
    dst=WakeUp
    rule=AnimGraph.AnimGraphNode_StateMachine_2.Locomotion.AnimStateTransitionNode_10.AnimationTransitionGraph_1
    src=Sleeping
    via=AnimStateTransitionNode_10
   }
   {
    dst=Fall
    rule=AnimGraph.AnimGraphNode_StateMachine_2.Locomotion.AnimStateTransitionNode_21.Transition
    src=Swimming
    via=AnimStateTransitionNode_21
   }
   {
    dst=Moving
    rule=AnimGraph.AnimGraphNode_StateMachine_2.Locomotion.AnimStateTransitionNode_25.Transition
    src=Swimming
    via=AnimStateTransitionNode_25
   }
   {
    dst=Idle
    rule=AnimGraph.AnimGraphNode_StateMachine_2.Locomotion.AnimStateTransitionNode_18.Transition
    src=WakeUp
    via=AnimStateTransitionNode_18
   }
  ]
 }
 Idle {
  path=AnimGraph.AnimGraphNode_StateMachine_2.Locomotion.AnimStateNode_0.Idle
  nodes [
   n1 = Blendspace Player 'Burt_Run_N_TurnPose_BS'!(X=Get IdleShuffleBodyBend() -> (IdleShuffleBodyBend:float)) -> (Pose) {blendSpace=&/Burt_Run_N_TurnPose_BS}
   n2 = Get IdleShufflePlayRate() -> (IdleShufflePlayRate:float)
   n3 = Layered blend per bone(
    BasePose=Use cached pose 'FaunaIdlePose'()
    BlendPoses_0=Blend Poses by bool(
     BlendPose_0=Sequence Player 'Burt_TurnInPlace_R_90_HipsAndFeet'!(PlayRate:float=@n2) -> (Pose) {sequence=&/Burt_TurnInPlace_R_90_HipsAndFeet}
     BlendPose_1=Sequence Player 'Burt_TurnInPlace_L_90_HipsAndFeet'!(PlayRate:float=@n2) -> (Pose) {sequence=&/Burt_TurnInPlace_L_90_HipsAndFeet}
     BlendTime_0=0.5
     BlendTime_1=0.5
     bActiveValue=Get bIdleShuffleTurnDirection() -> (bIdleShuffleTurnDirection:bool)
    ) {blendType=Linear, transitionType=Inertialization}
   ) {layerSetup [{branchFilters [{boneName=Leaf_Hips_Leaf_C}, {boneName=IK_Foot_Rear_Root}, {boneName=QuadSpine_A_Chest_C}, {boneName=IK_Foot_Front_Root}]}]}
   Output Animation Pose(Result=Blend Poses by bool(
     BlendPose_0=Apply Additive(Base=@n3, Additive=@n1) {alphaBoolBlend {blendOption=Linear}}
     BlendPose_1=Use cached pose 'FaunaIdlePose'()
     BlendTime_0=0.6
     BlendTime_1=0.6
     bActiveValue=Get bIsIdleShuffleTurn() -> (bIsIdleShuffleTurn:bool)
    ) {blendType=Linear, transitionType=Inertialization})
  ]
 }
 Moving {
  path=AnimGraph.AnimGraphNode_StateMachine_2.Locomotion.AnimStateNode_1.Moving
  nodes [
   n1 = Blendspace Player 'Burt_WalkRun_BS'!(X=Get Speed() -> (Speed:float), PlayRate:float=float / float(A=Get MovingPlayRate() -> (MovingPlayRate:float), B=1.2)) -> (Pose) {blendSpace=&/Burt_WalkRun_BS, groupName=Locomotion, method=SyncGroup}
   n2 = Blendspace Player 'Burt_Run_N_TurnPose_BS'(X=BreakRotator(InRot=Get PawnRotationRate() -> (PawnRotationRate:Rotator)).Yaw) {blendSpace=&Burt_Run_N_TurnPose_BS}
   AnimGraphNode_StateMachine_0 = JumpLand() {is=title AnimGraphNode_StateMachine}
   Output Animation Pose(Result=Apply Additive(Base=Apply Additive(Base=@n1, Additive=@n2) {alphaBoolBlend {blendOption=Linear}}, Additive=@AnimGraphNode_StateMachine_0) {alphaBoolBlend {blendOption=Linear}})
  ]
 }
 JumpLand {
  entry=AddRef
  path=AnimGraph.AnimGraphNode_StateMachine_2.Locomotion.AnimStateNode_1.Moving.AnimGraphNode_StateMachine_0.JumpLand
  ids=id
  nodes [
   AnimStateNode_0 = AddRef()
   AnimStateNode_1 = LandAdd() {stateLeft {guid=#1, notifyName=LandAddExit}}
   AnimStateTransitionNode_0 = AnimStateTransitionNode() {crossfadeDuration=0.1, logicType=TLT_Inertialization}
   AnimStateTransitionNode_1 = AnimStateTransitionNode() {bAutomaticRuleBasedOnSequencePlayerInState=true, crossfadeDuration=0.4, logicType=TLT_Inertialization}
  ]
  links [
   {
    dst=LandAdd
    rule=AnimGraph.AnimGraphNode_StateMachine_2.Locomotion.AnimStateNode_1.Moving.AnimGraphNode_StateMachine_0.JumpLand.AnimStateTransitionNode_0.Transition
    src=AddRef
    via=AnimStateTransitionNode_0
   }
   {
    dst=AddRef
    rule=AnimGraph.AnimGraphNode_StateMachine_2.Locomotion.AnimStateNode_1.Moving.AnimGraphNode_StateMachine_0.JumpLand.AnimStateTransitionNode_1.Transition
    src=LandAdd
    via=AnimStateTransitionNode_1
   }
  ]
 }
 AddRef {
  path=AnimGraph.AnimGraphNode_StateMachine_2.Locomotion.AnimStateNode_1.Moving.AnimGraphNode_StateMachine_0.JumpLand.AnimStateNode_0.AddRef
  nodes [Output Animation Pose(Result=Additive Identity Pose())]
 }
 LandAdd {
  path=AnimGraph.AnimGraphNode_StateMachine_2.Locomotion.AnimStateNode_1.Moving.AnimGraphNode_StateMachine_0.JumpLand.AnimStateNode_1.LandAdd
  nodes [Output Animation Pose(Result="Sequence Player (additive) 'Burt_JumpLand_Moving_Add'"() {sequence=&Burt_JumpLand_Moving_Add})]
 }
 Transition {
  path=AnimGraph.AnimGraphNode_StateMachine_2.Locomotion.AnimStateNode_1.Moving.AnimGraphNode_StateMachine_0.JumpLand.AnimStateTransitionNode_0.Transition
  nodes [Result(bCanEnterTransition=Get bPlayLandAdditive() -> (bPlayLandAdditive:bool))]
 }
 Transition {
  path=AnimGraph.AnimGraphNode_StateMachine_2.Locomotion.AnimStateNode_1.Moving.AnimGraphNode_StateMachine_0.JumpLand.AnimStateTransitionNode_1.Transition
  nodes [Result()]
 }
 Swimming {
  path=AnimGraph.AnimGraphNode_StateMachine_2.Locomotion.AnimStateNode_2.Swimming
  nodes [
   n1 = Sequence Player 'Burt_Swim_N'() {sequence=&/Burt_Swim_N}
   n2 = Sequence Player 'Burt_Swim_Idle'() {sequence=&/Burt_Swim_Idle}
   n3 = Get bIsMoving() -> (bIsMoving:bool)
   n4 = BreakRotator(InRot=Get PawnRotationRate())
   Output Animation Pose(Result=Apply Additive(
     Base=Blend Poses by bool(BlendPose_0=@n1, BlendPose_1=@n2, BlendTime_0=0.5, BlendTime_1=0.5, bActiveValue=@n3) {blendType=Linear, transitionType=Inertialization}
     Additive=Blendspace Player 'Burt_Run_N_TurnPose_BS'(X=@n4.Yaw) {blendSpace=&Burt_Run_N_TurnPose_BS}
    ) {alphaBoolBlend {blendOption=Linear}})
  ]
 }
 FallAsleep {
  path=AnimGraph.AnimGraphNode_StateMachine_2.Locomotion.AnimStateNode_3.FallAsleep
  nodes [Output Animation Pose(Result=Sequence Player 'Burt_LayingInto'() {bLoopAnimation=false, sequence=&/Burt_LayingInto})]
 }
 Sleeping {
  path=AnimGraph.AnimGraphNode_StateMachine_2.Locomotion.AnimStateNode_4.Sleeping
  nodes [Output Animation Pose(Result=Sequence Player 'Burt_LayingIdle'() {sequence=&/Burt_LayingIdle})]
 }
 JumpStart {
  path=AnimGraph.AnimGraphNode_StateMachine_2.Locomotion.AnimStateNode_5.JumpStart
  nodes [Output Animation Pose(Result=Sequence Player 'Burt_Jump_Start'() {bLoopAnimation=false, sequence=&/Burt_Jump_Start})]
 }
 Land {
  path=AnimGraph.AnimGraphNode_StateMachine_2.Locomotion.AnimStateNode_6.Land
  nodes [Output Animation Pose(Result=Sequence Player 'Burt_Jump_Land'() {bLoopAnimation=false, sequence=&/Burt_Jump_Land})]
 }
 WakeUp {
  path=AnimGraph.AnimGraphNode_StateMachine_2.Locomotion.AnimStateNode_7.WakeUp
  nodes [Output Animation Pose(Result=Sequence Player 'Burt_LayingExit'() {bLoopAnimation=false, sequence=&/Burt_LayingExit})]
 }
 Fall {
  path=AnimGraph.AnimGraphNode_StateMachine_2.Locomotion.AnimStateNode_8.Fall
  nodes [Output Animation Pose(Result=Sequence Player 'Burt_Jump_Loop'() {sequence=&/Burt_Jump_Loop})]
 }
 Transition {
  path=AnimGraph.AnimGraphNode_StateMachine_2.Locomotion.AnimStateTransitionNode_0.Transition
  nodes [Result(bCanEnterTransition=Get bCanTransitionIdleToMoving() -> (bCanTransitionIdleToMoving:bool))]
 }
 Transition {
  path=AnimGraph.AnimGraphNode_StateMachine_2.Locomotion.AnimStateTransitionNode_1.Transition
  nodes [Result(bCanEnterTransition=Get bCanTransitionIdleToMoving())]
 }
 AnimationTransitionGraph_1 {
  path=AnimGraph.AnimGraphNode_StateMachine_2.Locomotion.AnimStateTransitionNode_10.AnimationTransitionGraph_1
  nodes [Result(bCanEnterTransition=Get StateTransition_SleepingToWakeUp() -> (StateTransition_SleepingToWakeUp:bool))]
 }
 AnimationTransitionGraph_0 {
  path=AnimGraph.AnimGraphNode_StateMachine_2.Locomotion.AnimStateTransitionNode_12.AnimationTransitionGraph_0
  nodes [Result(bCanEnterTransition=Get bCanTransitionFallToIdle() -> (bCanTransitionFallToIdle:bool))]
 }
 Transition {path=AnimGraph.AnimGraphNode_StateMachine_2.Locomotion.AnimStateTransitionNode_13.Transition, nodes [Result()]}
 Transition {
  path=AnimGraph.AnimGraphNode_StateMachine_2.Locomotion.AnimStateTransitionNode_14.Transition
  nodes [Result(bCanEnterTransition=Get bIsJumping())]
 }
 Transition {
  path=AnimGraph.AnimGraphNode_StateMachine_2.Locomotion.AnimStateTransitionNode_15.Transition
  nodes [Result(bCanEnterTransition=Get bCanTransitionFallToMoving() -> (bCanTransitionFallToMoving:bool))]
 }
 Transition {
  path=AnimGraph.AnimGraphNode_StateMachine_2.Locomotion.AnimStateTransitionNode_16.Transition
  nodes [Result(bCanEnterTransition=Get bCanTransitionLandToJumpStart() -> (bCanTransitionLandToJumpStart:bool))]
 }
 Transition {
  path=AnimGraph.AnimGraphNode_StateMachine_2.Locomotion.AnimStateTransitionNode_17.Transition
  nodes [Result(bCanEnterTransition=Get StateTransition_IdleToFallAsleep() -> (StateTransition_IdleToFallAsleep:bool))]
 }
 Transition {
  path=AnimGraph.AnimGraphNode_StateMachine_2.Locomotion.AnimStateTransitionNode_18.Transition
  nodes [Result(bCanEnterTransition=Get StateTransition_WakeUpToIdle() -> (StateTransition_WakeUpToIdle:bool))]
 }
 Transition {path=AnimGraph.AnimGraphNode_StateMachine_2.Locomotion.AnimStateTransitionNode_19.Transition, nodes [Result()]}
 Transition {
  path=AnimGraph.AnimGraphNode_StateMachine_2.Locomotion.AnimStateTransitionNode_2.Transition
  nodes [Result(bCanEnterTransition=Get bIsSurfaceSwimming())]
 }
 Transition {
  path=AnimGraph.AnimGraphNode_StateMachine_2.Locomotion.AnimStateTransitionNode_21.Transition
  nodes [Result(bCanEnterTransition=Get bIsFalling())]
 }
 Transition {
  path=AnimGraph.AnimGraphNode_StateMachine_2.Locomotion.AnimStateTransitionNode_22.Transition
  nodes [Result(bCanEnterTransition=Get bIsSurfaceSwimming())]
 }
 Transition {
  path=AnimGraph.AnimGraphNode_StateMachine_2.Locomotion.AnimStateTransitionNode_25.Transition
  nodes [Result(bCanEnterTransition=NOT Boolean(A=Get bIsSurfaceSwimming()))]
 }
 Transition {
  path=AnimGraph.AnimGraphNode_StateMachine_2.Locomotion.AnimStateTransitionNode_26.Transition
  nodes [Result(bCanEnterTransition=Get bIsFalling())]
 }
 Transition {
  path=AnimGraph.AnimGraphNode_StateMachine_2.Locomotion.AnimStateTransitionNode_4.Transition
  nodes [Result(bCanEnterTransition=Get bCanTransitionMovingToIdle() -> (bCanTransitionMovingToIdle:bool))]
 }
 AnimationTransitionGraph_1 {
  path=AnimGraph.AnimGraphNode_StateMachine_2.Locomotion.AnimStateTransitionNode_5.AnimationTransitionGraph_1
  nodes [Result(bCanEnterTransition=Get StateTransition_FallAsleepToSleeping() -> (StateTransition_FallAsleepToSleeping:bool))]
 }
 Transition {
  path=AnimGraph.AnimGraphNode_StateMachine_2.Locomotion.AnimStateTransitionNode_6.Transition
  nodes [Result(bCanEnterTransition=Get bIsJumping())]
 }
 Transition {
  path=AnimGraph.AnimGraphNode_StateMachine_2.Locomotion.AnimStateTransitionNode_7.Transition
  nodes [Result(bCanEnterTransition=Get bIsSurfaceSwimming())]
 }
 Transition {
  path=AnimGraph.AnimGraphNode_StateMachine_2.Locomotion.AnimStateTransitionNode_8.Transition
  nodes [Result(bCanEnterTransition=Get bIsFalling())]
 }
 Transition {
  path=AnimGraph.AnimGraphNode_StateMachine_2.Locomotion.AnimStateTransitionNode_9.Transition
  nodes [Result(bCanEnterTransition=Get bIsSurfaceSwimming())]
 }
 RidingLocomotion {
  entry=Idle
  path=AnimGraph.AnimGraphNode_StateMachine_3.RidingLocomotion
  ids=id
  nodes [
   SA_ToMoving() {is=title AnimStateAliasNode, also {aliases [Idle,Stop,TurnInPlace,TurnSettle]}}
   SA_ToSwimming() {is=title AnimStateAliasNode, bGlobalAlias=true}
   SA_ToJumpStart() {is=title AnimStateAliasNode, also {aliases [Idle,JumpLand,Moving,Stop,TurnInPlace,TurnSettle]}}
   SA_ToFall() {is=title AnimStateAliasNode, also {aliases [Moving,Stop,TurnInPlace,TurnSettle]}}
   AnimStateNode_0 = Idle()
   AnimStateNode_1 = TurnInPlace() {is=title AnimStateNode}
   AnimStateNode_2 = Swimming()
   AnimStateNode_3 = Moving()
   AnimStateNode_4 = JumpLand() {is=title AnimStateNode, stateEntered {guid=#2, notifyName=LandEnter}}
   AnimStateNode_5 = JumpStart()
   AnimStateNode_6 = Stop() {is=title AnimStateNode}
   AnimStateNode_7 = Fall() {stateLeft {guid=#3, notifyName=FallExit}}
   AnimStateNode_8 = TurnSettle() {is=title AnimStateNode}
   AnimStateTransitionNode_0 = AnimStateTransitionNode() {logicType=TLT_Inertialization}
   AnimStateTransitionNode_1 = AnimStateTransitionNode() {crossfadeDuration=0.1, logicType=TLT_Inertialization, priorityOrder=2}
   AnimStateTransitionNode_11 = AnimStateTransitionNode()
   AnimStateTransitionNode_12 = AnimStateTransitionNode() {logicType=TLT_Inertialization, priorityOrder=2}
   AnimStateTransitionNode_13 = AnimStateTransitionNode() {bAutomaticRuleBasedOnSequencePlayerInState=true, crossfadeDuration=0.4, logicType=TLT_Inertialization}
   AnimStateTransitionNode_15 = AnimStateTransitionNode() {crossfadeDuration=0.4, logicType=TLT_Inertialization}
   AnimStateTransitionNode_17 = AnimStateTransitionNode() {crossfadeDuration=0.1, logicType=TLT_Inertialization}
   AnimStateTransitionNode_18 = AnimStateTransitionNode() {crossfadeDuration=0.3, logicType=TLT_Inertialization}
   AnimStateTransitionNode_19 = AnimStateTransitionNode() {bAutomaticRuleBasedOnSequencePlayerInState=true, crossfadeDuration=0.35, logicType=TLT_Inertialization, priorityOrder=100}
   AnimStateTransitionNode_2 = AnimStateTransitionNode() {crossfadeDuration=0.3, priorityOrder=2}
   AnimStateTransitionNode_20 = AnimStateTransitionNode() {logicType=TLT_Inertialization, priorityOrder=3}
   AnimStateTransitionNode_23 = AnimStateTransitionNode() {crossfadeDuration=0.4, logicType=TLT_Inertialization}
   AnimStateTransitionNode_24 = AnimStateTransitionNode() {bAutomaticRuleBasedOnSequencePlayerInState=true, crossfadeDuration=0.4, logicType=TLT_Inertialization, priorityOrder=3}
   AnimStateTransitionNode_25 = AnimStateTransitionNode() {logicType=TLT_Inertialization}
   AnimStateTransitionNode_3 = AnimStateTransitionNode() {crossfadeDuration=0.1}
   AnimStateTransitionNode_4 = AnimStateTransitionNode() {bAutomaticRuleBasedOnSequencePlayerInState=true, crossfadeDuration=0.3, priorityOrder=2}
   AnimStateTransitionNode_5 = AnimStateTransitionNode() {logicType=TLT_Inertialization}
   AnimStateTransitionNode_9 = AnimStateTransitionNode() {crossfadeDuration=0.4, logicType=TLT_Inertialization, priorityOrder=3}
  ]
  links [
   {
    dst=JumpLand
    rule=AnimGraph.AnimGraphNode_StateMachine_3.RidingLocomotion.AnimStateTransitionNode_12.AnimationTransitionGraph_0
    src=Fall
    via=AnimStateTransitionNode_12
   }
   {
    dst=Moving
    rule=AnimGraph.AnimGraphNode_StateMachine_3.RidingLocomotion.AnimStateTransitionNode_3.Transition
    src=Fall
    via=AnimStateTransitionNode_3
   }
   {
    dst=TurnInPlace
    rule=AnimGraph.AnimGraphNode_StateMachine_3.RidingLocomotion.AnimStateTransitionNode_11.Transition
    src=Idle
    via=AnimStateTransitionNode_11
   }
   {
    dst=Idle
    rule=AnimGraph.AnimGraphNode_StateMachine_3.RidingLocomotion.AnimStateTransitionNode_19.Transition
    src=JumpLand
    via=AnimStateTransitionNode_19
   }
   {
    dst=Moving
    rule=AnimGraph.AnimGraphNode_StateMachine_3.RidingLocomotion.AnimStateTransitionNode_1.Transition
    src=JumpLand
    via=AnimStateTransitionNode_1
   }
   {
    dst=Fall
    rule=AnimGraph.AnimGraphNode_StateMachine_3.RidingLocomotion.AnimStateTransitionNode_13.Transition
    src=JumpStart
    via=AnimStateTransitionNode_13
   }
   {
    dst=Stop
    rule=AnimGraph.AnimGraphNode_StateMachine_3.RidingLocomotion.AnimStateTransitionNode_20.Transition
    src=Moving
    via=AnimStateTransitionNode_20
   }
   {
    dst=Fall
    rule=AnimGraph.AnimGraphNode_StateMachine_3.RidingLocomotion.AnimStateTransitionNode_18.Transition
    src=SA_ToFall
    via=AnimStateTransitionNode_18
   }
   {
    dst=JumpStart
    rule=AnimGraph.AnimGraphNode_StateMachine_3.RidingLocomotion.AnimStateTransitionNode_17.Transition
    src=SA_ToJumpStart
    via=AnimStateTransitionNode_17
   }
   {
    dst=Moving
    rule=AnimGraph.AnimGraphNode_StateMachine_3.RidingLocomotion.AnimStateTransitionNode_9.Transition
    src=SA_ToMoving
    via=AnimStateTransitionNode_9
   }
   {
    dst=Swimming
    rule=AnimGraph.AnimGraphNode_StateMachine_3.RidingLocomotion.AnimStateTransitionNode_15.Transition
    src=SA_ToSwimming
    via=AnimStateTransitionNode_15
   }
   {
    dst=Idle
    rule=AnimGraph.AnimGraphNode_StateMachine_3.RidingLocomotion.AnimStateTransitionNode_24.Transition
    src=Stop
    via=AnimStateTransitionNode_24
   }
   {
    dst=Moving
    rule=AnimGraph.AnimGraphNode_StateMachine_3.RidingLocomotion.AnimStateTransitionNode_2.Transition
    src=Stop
    via=AnimStateTransitionNode_2
   }
   {
    dst=TurnInPlace
    rule=AnimGraph.AnimGraphNode_StateMachine_3.RidingLocomotion.AnimStateTransitionNode_23.Transition
    src=Stop
    via=AnimStateTransitionNode_23
   }
   {
    dst=Moving
    rule=AnimGraph.AnimGraphNode_StateMachine_3.RidingLocomotion.AnimStateTransitionNode_25.Transition
    src=Swimming
    via=AnimStateTransitionNode_25
   }
   {
    dst=TurnSettle
    rule=AnimGraph.AnimGraphNode_StateMachine_3.RidingLocomotion.AnimStateTransitionNode_0.Transition
    src=TurnInPlace
    via=AnimStateTransitionNode_0
   }
   {
    dst=Idle
    rule=AnimGraph.AnimGraphNode_StateMachine_3.RidingLocomotion.AnimStateTransitionNode_4.Transition
    src=TurnSettle
    via=AnimStateTransitionNode_4
   }
   {
    dst=TurnInPlace
    rule=AnimGraph.AnimGraphNode_StateMachine_3.RidingLocomotion.AnimStateTransitionNode_5.Transition
    src=TurnSettle
    via=AnimStateTransitionNode_5
   }
  ]
 }
 Idle {
  path=AnimGraph.AnimGraphNode_StateMachine_3.RidingLocomotion.AnimStateNode_0.Idle
  nodes [Output Animation Pose(Result=Sequence Player 'Burt_Idle_Riding'() {sequence=&/Burt_Idle_Riding})]
 }
 TurnInPlace {
  path=AnimGraph.AnimGraphNode_StateMachine_3.RidingLocomotion.AnimStateNode_1.TurnInPlace
  nodes [Output Animation Pose(Result=Blend Poses by bool(
     BlendPose_0=Sequence Player 'Burt_TurnRight_Riding'() {bLoopAnimation=false, sequence=&/Burt_TurnRight_Riding}
     BlendPose_1=Sequence Player 'Burt_TurnLeft_Riding'() {bLoopAnimation=false, sequence=&/Burt_TurnLeft_Riding}
     BlendTime_0=0.3
     BlendTime_1=0.3
    ) {blendType=Linear})]
 }
 Swimming {
  path=AnimGraph.AnimGraphNode_StateMachine_3.RidingLocomotion.AnimStateNode_2.Swimming
  nodes [
   n1 = Blend Poses by bool(
    BlendPose_0=Sequence Player 'Burt_Swim_S'() {groupName=Locomotion, method=SyncGroup, sequence=&/Burt_Swim_S}
    BlendPose_1=Sequence Player 'Burt_Swim_N'() {groupName=Locomotion, method=SyncGroup, sequence=&Burt_Swim_N}
    BlendTime_0=0.35
    BlendTime_1=0.35
    bActiveValue=Get bIsMovingBackwards() -> (bIsMovingBackwards:bool)
   ) {blendType=Linear, transitionType=Inertialization}
   n2 = Sequence Player 'Burt_Swim_Idle'() {groupName=Locomotion, method=SyncGroup, sequence=&Burt_Swim_Idle}
   n3 = Get bIsMoving()
   n4 = Get SwimmingYaw() -> (SwimmingYaw:float)
   Output Animation Pose(Result=Apply Additive(
     Base=Blend Poses by bool(BlendPose_0=@n1, BlendPose_1=@n2, BlendTime_0=0.5, BlendTime_1=1, bActiveValue=@n3) {blendType=Linear}
     Additive=Blendspace Player 'Burt_Run_N_TurnPose_BS'(X=@n4) {blendSpace=&Burt_Run_N_TurnPose_BS}
    ) {alphaBoolBlend {blendOption=Linear}})
  ]
 }
 Moving {
  path=AnimGraph.AnimGraphNode_StateMachine_3.RidingLocomotion.AnimStateNode_3.Moving
  nodes [
   n1 = Blendspace Player 'Burt_WalkRun_BS'(X=Get PawnSpeed2D(), PlayRate=Get RidingMovePlayRate() -> (RidingMovePlayRate:float)) {blendSpace=&Burt_WalkRun_BS, groupName=Locomotion, method=SyncGroup}
   n2 = Sequence Player 'Burt_Boost_N'() {groupName=Locomotion, method=SyncGroup, sequence=&/Burt_Boost_N}
   n3 = Break Riding Animation Data(RidingAnimationData=Get RidingAnimationData()) -> (
    bIsRidden=~
    bIsTurningInPlace=~
    bIsSprinting=~
    bIsBeingPetted=~
    bIsMovingBackwards=~
    TurnInPlaceRotationSpeed=~
    TurnInPlaceAngleDelta=~
    RiderReferentialYaw=~
   )
   n4 = BreakRotator(InRot=Get PawnRotationRate())
   n5 = Apply Additive(
    Base=Apply Additive(
     Base=Blend Poses by bool(BlendPose_0=@n2, BlendPose_1=@n1, BlendTime_0=0.25, BlendTime_1=0.25, bActiveValue=@n3.bIsSprinting) {blendType=Linear}
     Additive=Blendspace Player 'Burt_Run_N_TurnPose_BS'(X=@n4.Yaw) {blendSpace=&Burt_Run_N_TurnPose_BS}
    ) {alphaBoolBlend {blendOption=Linear}}
    Additive=Blendspace Player 'Burt_BodyRoll_BS'(X=Get BodyRoll() -> (BodyRoll:float), Y=Get PawnSpeed2D()) {blendSpace=&/Burt_BodyRoll_BS}
   ) {alphaBoolBlend {blendOption=Linear}}
   n6 = Apply Additive(
    Base=Sequence Player 'Burt_Walk_S'!(PlayRate:float=Get BackwardsPlayRate() -> (BackwardsPlayRate:float)) -> (Pose) {groupName=Locomotion, method=SyncGroup, sequence=&/Burt_Walk_S}
    Additive=Blendspace Player 'Burt_Walk_S_Add_BS'!(X=@n3.RiderReferentialYaw) -> (Pose) {blendSpace=&/Burt_Walk_S_Add_BS}
   ) {alphaBoolBlend {blendInTime=0.5, blendOption=Linear, blendOutTime=0.35}}
   AnimGraphNode_StateMachine_0 = JumpLand() {is=title AnimGraphNode_StateMachine}
   Output Animation Pose(Result=Apply Additive(
     Base=Blend Poses by bool(BlendPose_0=@n6, BlendPose_1=@n5, bActiveValue=@n3.bIsMovingBackwards) {blendType=Linear}
     Additive=@AnimGraphNode_StateMachine_0
    ) {alphaBoolBlend {blendOption=Linear}})
  ]
 }
 JumpLand {
  entry=AddRef
  path=AnimGraph.AnimGraphNode_StateMachine_3.RidingLocomotion.AnimStateNode_3.Moving.AnimGraphNode_StateMachine_0.JumpLand
  ids=id
  nodes [
   AnimStateNode_0 = AddRef()
   AnimStateNode_1 = LandAdd() {stateLeft {guid=#1, notifyName=LandAddExit}}
   AnimStateTransitionNode_0 = AnimStateTransitionNode() {crossfadeDuration=0.1, logicType=TLT_Inertialization}
   AnimStateTransitionNode_1 = AnimStateTransitionNode() {bAutomaticRuleBasedOnSequencePlayerInState=true, crossfadeDuration=0.4, logicType=TLT_Inertialization}
  ]
  links [
   {
    dst=LandAdd
    rule=AnimGraph.AnimGraphNode_StateMachine_3.RidingLocomotion.AnimStateNode_3.Moving.AnimGraphNode_StateMachine_0.JumpLand.AnimStateTransitionNode_0.Transition
    src=AddRef
    via=AnimStateTransitionNode_0
   }
   {
    dst=AddRef
    rule=AnimGraph.AnimGraphNode_StateMachine_3.RidingLocomotion.AnimStateNode_3.Moving.AnimGraphNode_StateMachine_0.JumpLand.AnimStateTransitionNode_1.Transition
    src=LandAdd
    via=AnimStateTransitionNode_1
   }
  ]
 }
 AddRef {
  path=AnimGraph.AnimGraphNode_StateMachine_3.RidingLocomotion.AnimStateNode_3.Moving.AnimGraphNode_StateMachine_0.JumpLand.AnimStateNode_0.AddRef
  nodes [Output Animation Pose(Result=Additive Identity Pose())]
 }
 LandAdd {
  path=AnimGraph.AnimGraphNode_StateMachine_3.RidingLocomotion.AnimStateNode_3.Moving.AnimGraphNode_StateMachine_0.JumpLand.AnimStateNode_1.LandAdd
  nodes [Output Animation Pose(Result="Sequence Player (additive) 'Burt_JumpLand_Moving_Add'"() {sequence=&Burt_JumpLand_Moving_Add})]
 }
 Transition {
  path=AnimGraph.AnimGraphNode_StateMachine_3.RidingLocomotion.AnimStateNode_3.Moving.AnimGraphNode_StateMachine_0.JumpLand.AnimStateTransitionNode_0.Transition
  nodes [Result(bCanEnterTransition=Get bPlayLandAdditive())]
 }
 Transition {
  path=AnimGraph.AnimGraphNode_StateMachine_3.RidingLocomotion.AnimStateNode_3.Moving.AnimGraphNode_StateMachine_0.JumpLand.AnimStateTransitionNode_1.Transition
  nodes [Result()]
 }
 JumpLand {
  path=AnimGraph.AnimGraphNode_StateMachine_3.RidingLocomotion.AnimStateNode_4.JumpLand
  nodes [Output Animation Pose(Result=Sequence Player 'Burt_Jump_Land'() {bLoopAnimation=false, sequence=&Burt_Jump_Land})]
 }
 JumpStart {
  path=AnimGraph.AnimGraphNode_StateMachine_3.RidingLocomotion.AnimStateNode_5.JumpStart
  nodes [Output Animation Pose(Result=Sequence Player 'Burt_Jump_Start'() {bLoopAnimation=false, sequence=&Burt_Jump_Start})]
 }
 Stop {
  path=AnimGraph.AnimGraphNode_StateMachine_3.RidingLocomotion.AnimStateNode_6.Stop
  nodes [
   AnimGraphNode_StateMachine_0 = StopTransition() {is=title AnimGraphNode_StateMachine, bAllowConduitEntryStates=true, bSkipFirstUpdateTransition=false, maxTransitionsPerFrame=1}
   Output Animation Pose(Result=@AnimGraphNode_StateMachine_0)
  ]
 }
 StopTransition {
  entry=Conduit
  path=AnimGraph.AnimGraphNode_StateMachine_3.RidingLocomotion.AnimStateNode_6.Stop.AnimGraphNode_StateMachine_0.StopTransition
  ids=id
  nodes [
   AnimStateConduitNode_0 = Conduit()
   AnimStateNode_0 = FeetInAir() {is=title AnimStateNode}
   AnimStateNode_1 = FrontFeetPlanted() {is=title AnimStateNode}
   AnimStateNode_2 = BackFeetPlanted() {is=title AnimStateNode}
   AnimStateNode_3 = Backwards() {is=title AnimStateNode}
   AnimStateTransitionNode_0 = AnimStateTransitionNode() {logicType=TLT_Inertialization, priorityOrder=0}
   AnimStateTransitionNode_7 = AnimStateTransitionNode() {logicType=TLT_Inertialization}
   AnimStateTransitionNode_8 = AnimStateTransitionNode() {logicType=TLT_Inertialization, priorityOrder=2}
   AnimStateTransitionNode_9 = AnimStateTransitionNode() {logicType=TLT_Inertialization, priorityOrder=100}
  ]
  links [
   {
    dst=BackFeetPlanted
    rule=AnimGraph.AnimGraphNode_StateMachine_3.RidingLocomotion.AnimStateNode_6.Stop.AnimGraphNode_StateMachine_0.StopTransition.AnimStateTransitionNode_9.Transition
    src=Conduit
    via=AnimStateTransitionNode_9
   }
   {
    dst=Backwards
    rule=AnimGraph.AnimGraphNode_StateMachine_3.RidingLocomotion.AnimStateNode_6.Stop.AnimGraphNode_StateMachine_0.StopTransition.AnimStateTransitionNode_0.Transition
    src=Conduit
    via=AnimStateTransitionNode_0
   }
   {
    dst=FeetInAir
    rule=AnimGraph.AnimGraphNode_StateMachine_3.RidingLocomotion.AnimStateNode_6.Stop.AnimGraphNode_StateMachine_0.StopTransition.AnimStateTransitionNode_7.Transition
    src=Conduit
    via=AnimStateTransitionNode_7
   }
   {
    dst=FrontFeetPlanted
    rule=AnimGraph.AnimGraphNode_StateMachine_3.RidingLocomotion.AnimStateNode_6.Stop.AnimGraphNode_StateMachine_0.StopTransition.AnimStateTransitionNode_8.Transition
    src=Conduit
    via=AnimStateTransitionNode_8
   }
  ]
 }
 Conduit {
  path=AnimGraph.AnimGraphNode_StateMachine_3.RidingLocomotion.AnimStateNode_6.Stop.AnimGraphNode_StateMachine_0.StopTransition.AnimStateConduitNode_0.Conduit
  nodes [Result(bCanEnterTransition=true)]
 }
 FeetInAir {
  path=AnimGraph.AnimGraphNode_StateMachine_3.RidingLocomotion.AnimStateNode_6.Stop.AnimGraphNode_StateMachine_0.StopTransition.AnimStateNode_0.FeetInAir
  nodes [Output Animation Pose(Result=Sequence Player 'Burt_Stop_FeetInAir'() {bLoopAnimation=false, sequence=&/Burt_Stop_FeetInAir})]
 }
 FrontFeetPlanted {
  path=AnimGraph.AnimGraphNode_StateMachine_3.RidingLocomotion.AnimStateNode_6.Stop.AnimGraphNode_StateMachine_0.StopTransition.AnimStateNode_1.FrontFeetPlanted
  nodes [Output Animation Pose(Result=Sequence Player 'Burt_Stop_FrontFeetPlanted'() {bLoopAnimation=false, sequence=&/Burt_Stop_FrontFeetPlanted})]
 }
 BackFeetPlanted {
  path=AnimGraph.AnimGraphNode_StateMachine_3.RidingLocomotion.AnimStateNode_6.Stop.AnimGraphNode_StateMachine_0.StopTransition.AnimStateNode_2.BackFeetPlanted
  nodes [Output Animation Pose(Result=Sequence Player 'Burt_Stop_BackFeetPlanted'() {bLoopAnimation=false, sequence=&/Burt_Stop_BackFeetPlanted})]
 }
 Backwards {
  path=AnimGraph.AnimGraphNode_StateMachine_3.RidingLocomotion.AnimStateNode_6.Stop.AnimGraphNode_StateMachine_0.StopTransition.AnimStateNode_3.Backwards
  nodes [
   AnimGraphNode_StateMachine_0 = StopBackwards() {is=title AnimGraphNode_StateMachine, bAllowConduitEntryStates=true, bSkipFirstUpdateTransition=false, maxTransitionsPerFrame=1}
   Output Animation Pose(Result=@AnimGraphNode_StateMachine_0)
  ]
 }
 StopBackwards {
  entry=Conduit
  path=AnimGraph.AnimGraphNode_StateMachine_3.RidingLocomotion.AnimStateNode_6.Stop.AnimGraphNode_StateMachine_0.StopTransition.AnimStateNode_3.Backwards.AnimGraphNode_StateMachine_0.StopBackwards
  ids=id
  nodes [
   AnimStateConduitNode_0 = Conduit()
   AnimStateNode_0 = RightPlantedLeftPass() {is=title AnimStateNode}
   AnimStateNode_1 = RightBackFeetForward() {is=title AnimStateNode}
   AnimStateNode_2 = LeftPlantedRightPass() {is=title AnimStateNode}
   AnimStateNode_3 = LeftBackFeetForward() {is=title AnimStateNode}
   AnimStateTransitionNode_1 = AnimStateTransitionNode() {logicType=TLT_Inertialization}
   AnimStateTransitionNode_2 = AnimStateTransitionNode() {logicType=TLT_Inertialization, priorityOrder=2}
   AnimStateTransitionNode_3 = AnimStateTransitionNode() {logicType=TLT_Inertialization, priorityOrder=3}
   AnimStateTransitionNode_4 = AnimStateTransitionNode() {logicType=TLT_Inertialization, priorityOrder=100}
  ]
  links [
   {
    dst=LeftBackFeetForward
    rule=AnimGraph.AnimGraphNode_StateMachine_3.RidingLocomotion.AnimStateNode_6.Stop.AnimGraphNode_StateMachine_0.StopTransition.AnimStateNode_3.Backwards.AnimGraphNode_StateMachine_0.StopBackwards.AnimStateTransitionNode_1.Transition
    src=Conduit
    via=AnimStateTransitionNode_1
   }
   {
    dst=LeftPlantedRightPass
    rule=AnimGraph.AnimGraphNode_StateMachine_3.RidingLocomotion.AnimStateNode_6.Stop.AnimGraphNode_StateMachine_0.StopTransition.AnimStateNode_3.Backwards.AnimGraphNode_StateMachine_0.StopBackwards.AnimStateTransitionNode_3.Transition
    src=Conduit
    via=AnimStateTransitionNode_3
   }
   {
    dst=RightBackFeetForward
    rule=AnimGraph.AnimGraphNode_StateMachine_3.RidingLocomotion.AnimStateNode_6.Stop.AnimGraphNode_StateMachine_0.StopTransition.AnimStateNode_3.Backwards.AnimGraphNode_StateMachine_0.StopBackwards.AnimStateTransitionNode_2.Transition
    src=Conduit
    via=AnimStateTransitionNode_2
   }
   {
    dst=RightPlantedLeftPass
    rule=AnimGraph.AnimGraphNode_StateMachine_3.RidingLocomotion.AnimStateNode_6.Stop.AnimGraphNode_StateMachine_0.StopTransition.AnimStateNode_3.Backwards.AnimGraphNode_StateMachine_0.StopBackwards.AnimStateTransitionNode_4.Transition
    src=Conduit
    via=AnimStateTransitionNode_4
   }
  ]
 }
 Conduit {
  path=AnimGraph.AnimGraphNode_StateMachine_3.RidingLocomotion.AnimStateNode_6.Stop.AnimGraphNode_StateMachine_0.StopTransition.AnimStateNode_3.Backwards.AnimGraphNode_StateMachine_0.StopBackwards.AnimStateConduitNode_0.Conduit
  nodes [Result(bCanEnterTransition=true)]
 }
 RightPlantedLeftPass {
  path=AnimGraph.AnimGraphNode_StateMachine_3.RidingLocomotion.AnimStateNode_6.Stop.AnimGraphNode_StateMachine_0.StopTransition.AnimStateNode_3.Backwards.AnimGraphNode_StateMachine_0.StopBackwards.AnimStateNode_0.RightPlantedLeftPass
  nodes [Output Animation Pose(Result=Sequence Player 'Burt_Walk_S_Stop_RPlantLPass'() {bLoopAnimation=false, sequence=&/Burt_Walk_S_Stop_RPlantLPass})]
 }
 RightBackFeetForward {
  path=AnimGraph.AnimGraphNode_StateMachine_3.RidingLocomotion.AnimStateNode_6.Stop.AnimGraphNode_StateMachine_0.StopTransition.AnimStateNode_3.Backwards.AnimGraphNode_StateMachine_0.StopBackwards.AnimStateNode_1.RightBackFeetForward
  nodes [Output Animation Pose(Result=Sequence Player 'Burt_Walk_S_Stop_RBackFtFwd'() {bLoopAnimation=false, sequence=&/Burt_Walk_S_Stop_RBackFtFwd})]
 }
 LeftPlantedRightPass {
  path=AnimGraph.AnimGraphNode_StateMachine_3.RidingLocomotion.AnimStateNode_6.Stop.AnimGraphNode_StateMachine_0.StopTransition.AnimStateNode_3.Backwards.AnimGraphNode_StateMachine_0.StopBackwards.AnimStateNode_2.LeftPlantedRightPass
  nodes [Output Animation Pose(Result=Sequence Player 'Burt_Walk_S_Stop_LPlantRPass'() {bLoopAnimation=false, sequence=&/Burt_Walk_S_Stop_LPlantRPass})]
 }
 LeftBackFeetForward {
  path=AnimGraph.AnimGraphNode_StateMachine_3.RidingLocomotion.AnimStateNode_6.Stop.AnimGraphNode_StateMachine_0.StopTransition.AnimStateNode_3.Backwards.AnimGraphNode_StateMachine_0.StopBackwards.AnimStateNode_3.LeftBackFeetForward
  nodes [Output Animation Pose(Result=Sequence Player 'Burt_Walk_S_Stop_LBackFtFwd'() {bLoopAnimation=false, sequence=&/Burt_Walk_S_Stop_LBackFtFwd})]
 }
 Transition {
  path=AnimGraph.AnimGraphNode_StateMachine_3.RidingLocomotion.AnimStateNode_6.Stop.AnimGraphNode_StateMachine_0.StopTransition.AnimStateNode_3.Backwards.AnimGraphNode_StateMachine_0.StopBackwards.AnimStateTransitionNode_1.Transition
  nodes [Result()]
 }
 Transition {
  path=AnimGraph.AnimGraphNode_StateMachine_3.RidingLocomotion.AnimStateNode_6.Stop.AnimGraphNode_StateMachine_0.StopTransition.AnimStateNode_3.Backwards.AnimGraphNode_StateMachine_0.StopBackwards.AnimStateTransitionNode_2.Transition
  nodes [Result()]
 }
 Transition {
  path=AnimGraph.AnimGraphNode_StateMachine_3.RidingLocomotion.AnimStateNode_6.Stop.AnimGraphNode_StateMachine_0.StopTransition.AnimStateNode_3.Backwards.AnimGraphNode_StateMachine_0.StopBackwards.AnimStateTransitionNode_3.Transition
  nodes [Result()]
 }
 Transition {
  path=AnimGraph.AnimGraphNode_StateMachine_3.RidingLocomotion.AnimStateNode_6.Stop.AnimGraphNode_StateMachine_0.StopTransition.AnimStateNode_3.Backwards.AnimGraphNode_StateMachine_0.StopBackwards.AnimStateTransitionNode_4.Transition
  nodes [Result(bCanEnterTransition=true)]
 }
 Transition {
  path=AnimGraph.AnimGraphNode_StateMachine_3.RidingLocomotion.AnimStateNode_6.Stop.AnimGraphNode_StateMachine_0.StopTransition.AnimStateTransitionNode_0.Transition
  nodes [Result(bCanEnterTransition=Get bShouldPlayStopBackwards() -> (bShouldPlayStopBackwards:bool))]
 }
 Transition {
  path=AnimGraph.AnimGraphNode_StateMachine_3.RidingLocomotion.AnimStateNode_6.Stop.AnimGraphNode_StateMachine_0.StopTransition.AnimStateTransitionNode_7.Transition
  nodes [Result()]
 }
 Transition {
  path=AnimGraph.AnimGraphNode_StateMachine_3.RidingLocomotion.AnimStateNode_6.Stop.AnimGraphNode_StateMachine_0.StopTransition.AnimStateTransitionNode_8.Transition
  nodes [Result()]
 }
 Transition {
  path=AnimGraph.AnimGraphNode_StateMachine_3.RidingLocomotion.AnimStateNode_6.Stop.AnimGraphNode_StateMachine_0.StopTransition.AnimStateTransitionNode_9.Transition
  nodes [Result(bCanEnterTransition=true)]
 }
 Fall {
  path=AnimGraph.AnimGraphNode_StateMachine_3.RidingLocomotion.AnimStateNode_7.Fall
  nodes [
   n1 = Blendspace Player 'Burt_SlopeSliding_BS'(X=Get SlopeSlidingRoll() -> (SlopeSlidingRoll:float), Y=Get SlopeSlidingPitch() -> (SlopeSlidingPitch:float)) {blendSpace=/Irwin/AI/Prey/Burt/Animations/SlopeSliding/Burt_SlopeSliding_BS}
   n2 = "Sequence Player (additive) 'HardSaddleBurt_SlopeSliding_Add'"() {sequence=/Irwin/AI/Prey/Burt/Animations/SlopeSliding/HardSaddleBurt_SlopeSliding_Add}
   Output Animation Pose(Result=Blend Poses by bool(
     BlendPose_0=Apply Additive(Base=@n1, Additive=@n2) {lODThreshold=1, alphaBoolBlend {blendOption=Linear}}
     BlendPose_1=Sequence Player 'Burt_Jump_Loop'() {sequence=&Burt_Jump_Loop}
     bActiveValue=Get bIsSlopeSliding() -> (bIsSlopeSliding:bool)
    ) {blendType=Linear})
  ]
 }
 TurnSettle {
  path=AnimGraph.AnimGraphNode_StateMachine_3.RidingLocomotion.AnimStateNode_8.TurnSettle
  nodes [Output Animation Pose(Result=Blend Poses by bool(
     BlendPose_0=Sequence Player 'Burt_TurnRight_Riding_Settle'() {bLoopAnimation=false, sequence=&/Burt_TurnRight_Riding_Settle}
     BlendPose_1=Sequence Player 'Burt_TurnLeft_Riding_Settle'() {bLoopAnimation=false, sequence=&/Burt_TurnLeft_Riding_Settle}
     BlendTime_0=0.3
     BlendTime_1=0.3
     bActiveValue=Get bWasTurningRight() -> (bWasTurningRight:bool)
    ) {blendType=Linear})]
 }
 Transition {
  path=AnimGraph.AnimGraphNode_StateMachine_3.RidingLocomotion.AnimStateTransitionNode_0.Transition
  nodes [
   n1 = Get RidingAnimationData()
   Result(bCanEnterTransition=NOT Boolean(A=Break Riding Animation Data(RidingAnimationData=@n1) -> (
      bIsRidden=~
      bIsTurningInPlace=~
      bIsSprinting=~
      bIsBeingPetted=~
      bIsMovingBackwards=~
      TurnInPlaceRotationSpeed=~
      TurnInPlaceAngleDelta=~
      RiderReferentialYaw=~
     ).bIsTurningInPlace))
  ]
 }
 Transition {
  path=AnimGraph.AnimGraphNode_StateMachine_3.RidingLocomotion.AnimStateTransitionNode_1.Transition
  nodes [Result(bCanEnterTransition=Get bCanTransitionIdleToMoving())]
 }
 Transition {
  path=AnimGraph.AnimGraphNode_StateMachine_3.RidingLocomotion.AnimStateTransitionNode_11.Transition
  nodes [Result(bCanEnterTransition=Break Riding Animation Data(RidingAnimationData=Get RidingAnimationData()) -> (
     bIsRidden=~
     bIsTurningInPlace=~
     bIsSprinting=~
     bIsBeingPetted=~
     bIsMovingBackwards=~
     TurnInPlaceRotationSpeed=~
     TurnInPlaceAngleDelta=~
     RiderReferentialYaw=~
    ).bIsTurningInPlace)]
 }
 AnimationTransitionGraph_0 {
  path=AnimGraph.AnimGraphNode_StateMachine_3.RidingLocomotion.AnimStateTransitionNode_12.AnimationTransitionGraph_0
  nodes [Result(bCanEnterTransition=Get bCanTransitionFallToIdle())]
 }
 Transition {path=AnimGraph.AnimGraphNode_StateMachine_3.RidingLocomotion.AnimStateTransitionNode_13.Transition, nodes [Result()]}
 Transition {
  path=AnimGraph.AnimGraphNode_StateMachine_3.RidingLocomotion.AnimStateTransitionNode_15.Transition
  nodes [Result(bCanEnterTransition=Get bIsSurfaceSwimming())]
 }
 Transition {
  path=AnimGraph.AnimGraphNode_StateMachine_3.RidingLocomotion.AnimStateTransitionNode_17.Transition
  nodes [Result(bCanEnterTransition=Get bIsJumping())]
 }
 Transition {
  path=AnimGraph.AnimGraphNode_StateMachine_3.RidingLocomotion.AnimStateTransitionNode_18.Transition
  nodes [Result(bCanEnterTransition=Get bIsFalling())]
 }
 Transition {path=AnimGraph.AnimGraphNode_StateMachine_3.RidingLocomotion.AnimStateTransitionNode_19.Transition, nodes [Result()]}
 Transition {
  path=AnimGraph.AnimGraphNode_StateMachine_3.RidingLocomotion.AnimStateTransitionNode_2.Transition
  nodes [Result(bCanEnterTransition=Get bCanTransitionIdleToMoving())]
 }
 Transition {
  path=AnimGraph.AnimGraphNode_StateMachine_3.RidingLocomotion.AnimStateTransitionNode_20.Transition
  nodes [Result(bCanEnterTransition=Get bCanTransitionMovingToIdle())]
 }
 Transition {
  path=AnimGraph.AnimGraphNode_StateMachine_3.RidingLocomotion.AnimStateTransitionNode_23.Transition
  nodes [Result(bCanEnterTransition=Break Riding Animation Data(RidingAnimationData=Get RidingAnimationData()) -> (
     bIsRidden=~
     bIsTurningInPlace=~
     bIsSprinting=~
     bIsBeingPetted=~
     bIsMovingBackwards=~
     TurnInPlaceRotationSpeed=~
     TurnInPlaceAngleDelta=~
     RiderReferentialYaw=~
    ).bIsTurningInPlace)]
 }
 Transition {path=AnimGraph.AnimGraphNode_StateMachine_3.RidingLocomotion.AnimStateTransitionNode_24.Transition, nodes [Result()]}
 Transition {
  path=AnimGraph.AnimGraphNode_StateMachine_3.RidingLocomotion.AnimStateTransitionNode_25.Transition
  nodes [Result(bCanEnterTransition=NOT Boolean(A=Get bIsSurfaceSwimming()))]
 }
 Transition {
  path=AnimGraph.AnimGraphNode_StateMachine_3.RidingLocomotion.AnimStateTransitionNode_3.Transition
  nodes [Result(bCanEnterTransition=Get bCanTransitionFallToMoving())]
 }
 Transition {path=AnimGraph.AnimGraphNode_StateMachine_3.RidingLocomotion.AnimStateTransitionNode_4.Transition, nodes [Result()]}
 Transition {
  path=AnimGraph.AnimGraphNode_StateMachine_3.RidingLocomotion.AnimStateTransitionNode_5.Transition
  nodes [Result(bCanEnterTransition=Break Riding Animation Data(RidingAnimationData=Get RidingAnimationData()) -> (
     bIsRidden=~
     bIsTurningInPlace=~
     bIsSprinting=~
     bIsBeingPetted=~
     bIsMovingBackwards=~
     TurnInPlaceRotationSpeed=~
     TurnInPlaceAngleDelta=~
     RiderReferentialYaw=~
    ).bIsTurningInPlace)]
 }
 Transition {
  path=AnimGraph.AnimGraphNode_StateMachine_3.RidingLocomotion.AnimStateTransitionNode_9.Transition
  nodes [Result(bCanEnterTransition=Get bCanTransitionIdleToMoving())]
 }
 EventGraph {nodes [
   AnimNotify_FallExit() {is=title K2Node_Event}
   > Set bPlayLandAdditive(bPlayLandAdditive:bool=true) -> (Output_Get:bool)
   AnimNotify_LandAddExit() {is=title K2Node_Event}
   > Set bPlayLandAdditive(bPlayLandAdditive=false)
   AnimNotify_LandEnter() {is=title K2Node_Event}
   > Set bPlayLandAdditive(bPlayLandAdditive=false)
  ]}
 Fauna Idle {nodes [
   Input Pose() {name=InPose Idle}
   Output Pose(Result=Sequence Player 'Burt_Idle_Short'() {sequence=&/Burt_Idle_Short}) {layerGroup=FaunaIdleOverride}
  ]}
 Fauna IdleTurn {nodes [
   Input Pose() {name=InPose IdleTurn}
   Output Pose(Result=Blend Poses by bool(
     BlendPose_0=Sequence Player 'Burt_TurnLeft'() {bLoopAnimation=false, groupName=Locomotion, method=SyncGroup, sequence=&/Burt_TurnLeft}
     BlendPose_1=Sequence Player 'Burt_TurnRight'() {bLoopAnimation=false, groupName=Locomotion, method=SyncGroup, sequence=&/Burt_TurnRight}
     bActiveValue=Get bTurnAngleIsNegative() -> (bTurnAngleIsNegative:bool)
    ) {blendType=Linear, transitionType=Inertialization}) {layerGroup=FaunaIdleOverride}
  ]}
 GetSequencerAnimSlotNames {nodes [
   GetSequencerAnimSlotNames() {is=title K2Node_FunctionEntry}
   > Return Node!(execute, ReturnValue:Array of Names=Make Array("[0]":Name=SequencerDefault) -> (Array:Array of Names))
  ]}
]
variables []
