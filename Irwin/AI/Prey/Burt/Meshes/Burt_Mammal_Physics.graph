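# Physics asset for Burt, a four-legged mammal prey AI, giving its skeleton a simplified collision shape.
# Eight single-capsule bodies cover the head, chest, pelvis, both elbows, both knees and the third tail bone, with left and right limb capsules mirrored.
# No constraints join the bodies, so the asset defines per-bone collision volumes without any jointed ragdoll setup.
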PhysicsAsset /Irwin/AI/Prey/Burt/Meshes/Burt_Mammal_Physics
bodies {
 BipedHeadNeck_A_Head_C [{shapeType=Capsule, length=29.34, radius=14.29, center=<10.4237 4.4991 0.4266>, rotation=<rot 1.01 141.2803 -87.8197>}]
 PawedArm_A_Elbow_L [{shapeType=Capsule, length=38.25, radius=8.75, center=<-0.38 0.06 -2.32>, rotation=<rot 84.99 169.99 -179.99>}]
 PawedArm_A_Elbow_R [{shapeType=Capsule, length=38.25, radius=8.75, center=<0.38 -0.06 2.32>, rotation=<rot 84.9898 169.99 0.01>}]
 PawedLeg_A_Knee_L [{shapeType=Capsule, length=25, radius=15, center=<14.73 4.51 -5.44>, rotation=<rot 22.51 -62.77 -83.84>}]
 PawedLeg_A_Knee_R [{shapeType=Capsule, length=25, radius=15, center=<-14.73 -4.51 5.44>, rotation=<rot 22.51 -62.77 96.16>}]
 QuadSpine_A_Chest_C [{shapeType=Capsule, length=14.75, radius=29, center=<-2.8579 1.8197 -0.0011>, rotation=<rot 90 0 50.005>}]
 QuadSpine_A_Pelvis_C [{shapeType=Capsule, length=29.5, radius=23, center=<14.9309 1.3716 0>, rotation=<rot 90 0 10.0008>}]
 Tail_A_Tail3_C [{shapeType=Capsule, length=39.25, radius=7.25, center=<4.1319 0 0>, rotation=<rot 90 0 0>}]
}
constraints []
